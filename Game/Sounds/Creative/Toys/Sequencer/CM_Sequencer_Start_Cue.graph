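# Sound cue that, judging by its name and folder, is the start sound for the Creative mode Sequencer toy.
# A single wave player is the root and plays the CM_Sequencer_Start01 wave, with no randomization or mixing nodes.
# It uses the gadget MoveTool attenuation asset for spatial falloff and is routed to the Creative gadgets sound class.
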SoundCue /Game/Sounds/Creative/Toys/Sequencer/CM_Sequencer_Start_Cue
root=WavePlayer_2
nodes [{id=WavePlayer_2, properties {soundWaveAssetPtr=&/CM_Sequencer_Start01}}]
properties {
 attenuationSettings=/Game/Sounds/Creative/Attenuations/Fort_Gadget_MoveTool_Att
 maxDistance=4000
 soundClassObject=&Fort_Gadgets_Creative
 volumeMultiplier=1
}
